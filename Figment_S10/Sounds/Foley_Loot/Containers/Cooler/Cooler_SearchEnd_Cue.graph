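# Sound cue for the end of a cooler container search, played as attenuated looting foley with a concurrency override.
# A quality switch selects between two layered mixes and two lighter branches that each play a single pre-made "opt" wave.
# The mixes layer the no-more-loot UI sound with a random cooler search wave shaped by a short volume envelope and pitch/volume modulation.
# The second mix's random node has one empty slot in place of a third wave.
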SoundCue /Figment_S10/Sounds/Foley_Loot/Containers/Cooler/Cooler_SearchEnd_Cue
root=QualityLevel_0
nodes [
 {
  id=Enveloper_0
  children [Random_0]
  properties {pitchCurve {editorCurveData {keys [{time=0}]}}, volumeCurve {editorCurveData {keys [<at 0 0>,<at 0.0569 0.9114>,<at 0.65 0>]}}}
 }
 {
  id=Enveloper_5
  children [Random_44]
  properties {pitchCurve {editorCurveData {keys [{time=0}]}}, volumeCurve {editorCurveData {keys [<at 0 0>,<at 0.0569 0.9114>,<at 0.65 0>]}}}
 }
 {id=Mixer_0, children [WavePlayer_0,Modulator_0], properties {inputVolume [0.8,0.5]}}
 {id=Mixer_12, children [WavePlayer_43, Modulator_15], properties {inputVolume [0.8,0.5]}}
 {id=Modulator_0, children [Enveloper_0], properties {pitchMax=1.07, pitchMin=0.92, volumeMax=1.07, volumeMin=0.92}}
 {id=Modulator_1, children [WavePlayer_3], properties {pitchMax=1.07, pitchMin=0.92, volumeMax=1.07, volumeMin=0.92}}
 {id=Modulator_15, children [Enveloper_5], properties {pitchMax=1.07, pitchMin=0.92, volumeMax=1.07, volumeMin=0.92}}
 {id=Modulator_2, children [WavePlayer_4], properties {pitchMax=1.07, pitchMin=0.92, volumeMax=1.07, volumeMin=0.92}}
 {id=QualityLevel_0, children [Mixer_12,Mixer_0,Modulator_1,Modulator_2]}
 {id=Random_0, children [WavePlayer_1,WavePlayer_2,null], properties {weights [1,1,1]}}
 {id=Random_44, children [WavePlayer_276, WavePlayer_277, WavePlayer_279], properties {weights [1,1,1]}}
 {id=WavePlayer_0, properties {soundWaveAssetPtr=/Game/Sounds/Fort_UI_Sounds/Gameplay/Fort_UI_NoMoreLoot}}
 {id=WavePlayer_1, properties {soundWaveAssetPtr=/Game/Sounds/Foley_Loot/Containers/Cooler/Cooler_Search08}}
 {id=WavePlayer_2, properties {soundWaveAssetPtr=/Game/Sounds/Foley_Loot/Containers/Cooler/Cooler_Search01}}
 {id=WavePlayer_276, properties {soundWaveAssetPtr=&Cooler_Search08}}
 {id=WavePlayer_277, properties {soundWaveAssetPtr=&Cooler_Search01}}
 {id=WavePlayer_279, properties {soundWaveAssetPtr=/Game/Sounds/Foley_Loot/Containers/Cooler/Cooler_Search04}}
 {id=WavePlayer_3, properties {soundWaveAssetPtr=/Game/Sounds/Foley_Loot/Containers/Cooler/Cooler_SearchEnd_opt}}
 {id=WavePlayer_4, properties {soundWaveAssetPtr=&Cooler_SearchEnd_opt}}
 {id=WavePlayer_43, properties {soundWaveAssetPtr=&Fort_UI_NoMoreLoot}}
]
properties {
 attenuationSettings=/Game/Sounds/Attenuation/Fort_Prop_Search_Att
 bOverrideConcurrency=true
 maxDistance=1300
 soundClassObject=/Figment_S10/Sounds/FortSoundClassesAndModes/Fort_Searching_Foley
 volumeMultiplier=0.7
 concurrencyOverrides {maxCount=4, resolutionRule=StopFarthestThenPreventNew}
}
